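# Physics asset for the M_MED_Tactical_Biker backpack cosmetic, defining collision and joints for its dangling rings and zipper.
# A capsule covers the pack, and small boxes cover the dyn_ bones.
# Constraints form a two-link ring chain from dyn_ring_mstr with wide swing and moderate twist, plus a more tightly limited zipper joint on dyn_zipper_mstr.
# No constraint links either master bone to the pack body.
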PhysicsAsset /BRCosmetics/Accessories/FORT_Backpacks/M_MED_Tactical_Biker/Meshes/M_MED_Tactical_Biker_Backpack_PhysicsAsset
bodies {
 dyn_ring [{shapeType=Box, extentX=3.8283, extentY=1.1777, extentZ=3.6829, center=<1.6562 0.0156 0.1788>}]
 dyn_ring02 [{shapeType=Box, extentX=7.7439, extentY=4.6693, extentZ=6.3416, center=<3.4168 -0.0734 -0.9436>}]
 dyn_ring_mstr [{shapeType=Box, extentX=2, extentY=2, extentZ=2}]
 dyn_zipper [{shapeType=Box, extentX=2, extentY=1.01, extentZ=1.01, center=<1 0 0>}]
 dyn_zipper_mstr [{shapeType=Box, extentX=0.25, extentY=0.25, extentZ=0.25}]
 pack [{shapeType=Capsule, length=11.9764, radius=11.7693, center=<0 0 3.4331>, rotation=<rot 0 -90 0>}]
}
constraints [
 {bone1Name=dyn_ring, bone2Name=dyn_ring_mstr, swing1LimitDegrees=65, swing2LimitDegrees=65, twistLimitDegrees=25}
 {bone1Name=dyn_ring02, bone2Name=dyn_ring, swing1LimitDegrees=65, swing2LimitDegrees=65, twistLimitDegrees=25}
 {bone1Name=dyn_zipper, bone2Name=dyn_zipper_mstr, swing1LimitDegrees=45, swing2LimitDegrees=5, twistLimitDegrees=1}
]
